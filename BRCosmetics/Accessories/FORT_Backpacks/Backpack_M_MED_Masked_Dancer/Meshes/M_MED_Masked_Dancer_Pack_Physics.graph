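# Physics asset for the Masked Dancer backpack cosmetic, defining bodies and joints for its dangling chains, flap and zipper.
# Upper and lower two-link chains, the flap and the zipper each hang from a Dyn_Master anchor bone through angular-limited constraints.
# Flap and zipper lock twist, the flap also locking one swing axis.
# The flap master joins the upper-chain master with free rotation; that chain master also carries a large box shape.
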PhysicsAsset /BRCosmetics/Accessories/FORT_Backpacks/Backpack_M_MED_Masked_Dancer/Meshes/M_MED_Masked_Dancer_Pack_Physics
bodies {
 Dyn_Chain_Lower_1 [{shapeType=Capsule, length=2.5, radius=0.5, center=<2 0 0>, rotation=<rot 90 0 0>}]
 Dyn_Chain_Lower_2 [{shapeType=Capsule, length=2, radius=0.505, center=<0.5 0 0>, rotation=<rot 0 0 90>}]
 Dyn_Chain_Upper_1 [{shapeType=Capsule, length=2, radius=1, center=<2.416 0.1252 -0.2427>, rotation=<rot 0 -0.9041 -89.0932>}]
 Dyn_Chain_Upper_2 [{shapeType=Capsule, length=0.5, radius=0.5, center=<0.8522 0 0>, rotation=<rot 90 0 0>}]
 Dyn_Flap [{shapeType=Capsule, length=5, radius=1, center=<1 0 0>, rotation=<rot 0 21.4223 89.9998>}]
 Dyn_Master_Chain_1 [
  {shapeType=Capsule, length=0.5, radius=0.5}
  {shapeType=Box, extentX=1, extentY=10, extentZ=8, center=<0.1209 0 5.2155>, rotation=<rot -19.7364 0 0>}
 ]
 Dyn_Master_Chain_2 [{shapeType=Capsule, length=0.5, radius=0.5}]
 Dyn_Master_Flap [{shapeType=Capsule, length=0.505, radius=0.505, rotation=<rot 0 -90 0>}]
 Dyn_Master_Zipper [{shapeType=Capsule, length=0.505, radius=0.505, rotation=<rot 0 -90 0>}]
 Dyn_Zipper [{shapeType=Capsule, length=0.505, radius=0.505, center=<1.5854 -0.0236 0.0762>, rotation=<rot 0 87.4231 -89.1831>}]
 pack [{shapeType=Capsule, length=0.5, radius=0.5}]
}
constraints [
 {bone1Name=Dyn_Chain_Lower_1, bone2Name=Dyn_Master_Chain_2, swing1LimitDegrees=45, swing2LimitDegrees=45, twistLimitDegrees=45}
 {bone1Name=Dyn_Chain_Lower_2, bone2Name=Dyn_Chain_Lower_1, swing1LimitDegrees=45, swing2LimitDegrees=45, twistLimitDegrees=45}
 {bone1Name=Dyn_Chain_Upper_2, bone2Name=Dyn_Chain_Upper_1, swing1LimitDegrees=45, swing2LimitDegrees=45, twistLimitDegrees=45}
 {bone1Name=Dyn_Chain_Upper_1, bone2Name=Dyn_Master_Chain_1, swing1LimitDegrees=2, swing2LimitDegrees=20, twistLimitDegrees=2}
 {bone1Name=Dyn_Zipper, bone2Name=Dyn_Master_Zipper, swing1LimitDegrees=10, swing2LimitDegrees=30, twistLimitDegrees=45, twistMotion=Locked}
 {
  bone1Name=Dyn_Master_Flap
  bone2Name=Dyn_Master_Chain_1
  swing1LimitDegrees=45
  swing1Motion=Free
  swing2LimitDegrees=45
  swing2Motion=Free
  twistLimitDegrees=45
  twistMotion=Free
 }
 {
  bone1Name=Dyn_Flap
  bone2Name=Dyn_Master_Flap
  swing1LimitDegrees=10
  swing1Motion=Locked
  swing2LimitDegrees=30
  twistLimitDegrees=45
  twistMotion=Locked
 }
]
